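# Unlit Sparks stage screen material showing an audio-reactive castle, terrain and fire scene, with parameters for horizontal-screen layout and a castle-window fire variant.
# Channel-packed castle and terrain textures are scaled, offset and mask-composited, with panning noise driving fire gradients and the terrain color wash.
# MPC beat and measure lifetimes pulse fire and brightness, and colors blend between parameters and stage colors.
# The result passes through MF_Sparks_Screen_Details as emissive.
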Material /SparksMaps/Effects/Screens/Castle/M_Sparks_Screen_Castle
graph {
 properties {bAutomaticallySetUsageInEditor=false, bHasPixelAnimation=true, shadingModel=MSM_Unlit, usedShadingModels=Unlit}
 nodes [
  n1 = LinearInterpolate(
   A=TextureSample(UVs=NamedRerouteUsage()) {texture=&/T_Sparks_Screens_Castle_TexPack}.A
   Alpha=NamedRerouteUsage() {desc=Masks Out Terrain for Horz Screen}
  ) {constB=0}
  n2 = NamedRerouteUsage()
  n3 = ScalarParameter() {defaultValue=2, desc=BeatLow, parameterName=Param_1}
  n4 = ScalarParameter() {desc=BeatHigh}
  n5 = CollectionParameter() {collection=/FMAudioReactiveData/AudioAnalysis/MPC_Sparks_Audio_Parameters, parameterName=BeatLifetime}
  n6 = NamedRerouteUsage()
  n7 = Power(
   Base=MaterialFunctionCall() {materialFunction=&LinearGradient}.VGradient
   Exp=ScalarParameter() {defaultValue=3, parameterName=GradientBackCasstle}
  )
  n8 = ScalarParameter() {defaultValue=0.5, parameterName=BackCastleBright}
  n9 = ScalarParameter() {parameterName=HorizontalScreen?}
  n10 = Multiply(
   A=AppendVector(A=LinearInterpolate(Alpha=@n9) {constA=2, constAlpha=0, constB=-2}, B=Constant() {r=1}) {desc=Scale}
   B=ScalarParameter() {defaultValue=0.6, parameterName=ScaleByCenter}
  )
  n11 = Add(
   A=NamedRerouteDeclaration(None=MaterialFunctionCall(Texture Scale=@n10) {materialFunction=&ScaleUVsByCenter}.UVs) {name=TextureScale_1, nodeColor=<rgba 1 0.5581 0 1>}
   B=AppendVector(A=ScalarParameter() {defaultValue=-0.225, desc=Slide, parameterName=SlideSideways}, B=ScalarParameter() {parameterName=SlideUpDown})
  ) {desc=Position}
  n12 = Add(A=AppendVector(A=Constant() {r=0.13}, B=Constant() {r=-0.2}) {desc=HorizontalCastleNudge}, B=@n11)
  n13 = ScalarParameter() {parameterName=HorizontalScreen?}
  n14 = TextureSample(UVs=NamedRerouteDeclaration(None=LinearInterpolate(A=@n11, B=@n12, Alpha=@n13)) {name=CenterScreenNudge, nodeColor=<rgba 0 1 0.2266 1>}) {texture=&T_Sparks_Screens_Castle_TexPack}
  n15 = LinearInterpolate(
   A=Multiply(A=Multiply(A=@n7, B=@n8) {desc=BackCastleGradient}, B=@n14.G) {constB=0.05}
   B=Saturate(None=Add(A=@n14.R) {constB=0.05})
   Alpha=@n14.B
  )
  n16 = CollectionParameter() {collection=&MPC_Sparks_Audio_Parameters, parameterName=MeasureLifetime}
  n17 = Add(
   A=NamedRerouteUsage()
   B=Multiply(A=Multiply(A=@n6, B=@n15), B=LinearInterpolate(Alpha=@n16) {constA=4, constB=2, desc=CastleBrightness + Beat})
  )
  n18 = TextureSample(UVs=@n11) {texture=&/T_Sparks_Screens_Castle_Terrain_TexPack}
  n19 = NamedRerouteUsage()
  n20 = ScalarParameter() {defaultValue=0.5, parameterName=TerrainGLow}
  n21 = Add(
   A=Multiply(A=Multiply(A=@n18.R, B=@n19), B=ScalarParameter() {defaultValue=1, parameterName=TerrainColorBright})
   B=Multiply(A=@n19, B=Multiply(A=@n18.G, B=@n20))
  )
  n22 = NamedRerouteDeclaration(None=@n18.A) {name=TerrainMaskCenter, nodeColor=<rgba 0 0.0391 1 1>}
  n23 = Add(A=NamedRerouteUsage(), B=NamedRerouteUsage())
  n24 = Add(
   A=Multiply(A=Multiply(A=@n1, B=@n2), B=LinearInterpolate(A=@n3, B=@n4, Alpha=@n5) {desc=BeatData})
   B=LinearInterpolate(A=NamedRerouteUsage(), B=LinearInterpolate(A=@n17, B=@n21, Alpha=@n22) {desc=Castle + Terrain combined}, Alpha=Saturate(None=@n23))
  )
  n25 = NamedRerouteUsage() {desc=For Window Fire}
  n26 = ScalarParameter() {parameterName=CastleWindows?}
  MaterialFunctionCall!(
   Emissive=Multiply(A=LinearInterpolate(A=@n24, B=@n25, Alpha=@n26), B=ScalarParameter() {defaultValue=0.4, parameterName=OverallBrightness})
   Scanlines Tile
  ) {materialFunction=/SparksMaps/Effects/Screens/Materials/MF/MF_Sparks_Screen_Details}
  n27 = SmoothStep(
   Min=ScalarParameter() {parameterName=FireBack_Gradient_Min}
   Max=ScalarParameter() {defaultValue=0.5, parameterName=FireBack_Gradient_Max}
   Value=ComponentMask(None=TextureCoordinate()) {g=true}
  ) {constMax=0.8, constMin=0.6}
  n28 = NamedRerouteUsage()
  n29 = LinearInterpolate(A=Multiply(A=ScalarParameter() {defaultValue=0.2, parameterName=FireBack_Black}, B=@n28), B=@n28, Alpha=NamedRerouteUsage())
  n30 = TextureCoordinate()
  n31 = Multiply(
   A=SmoothStep(
    Min=ScalarParameter() {defaultValue=0.4, parameterName=FireBack_Glow_Min}
    Max=ScalarParameter() {defaultValue=1, parameterName=FireBack_Glow_Max}
    Value=ComponentMask(None=@n30) {g=true}
   ) {constMax=0.8, constMin=0.6}
   B=ScalarParameter() {defaultValue=7, parameterName=FireBack_Glow_Power}
  )
  n32 = Multiply(A=@n28, B=@n31) {desc=Remove for White Glow}
  n33 = Multiply(A=@n31, B=ScalarParameter() {defaultValue=0.1, parameterName=FireBack_White}) {constB=0.01}
  NamedRerouteDeclaration(None=Add(A=Multiply(A=@n27, B=@n29), B=Add(A=@n32, B=@n33))) {name=BG_Fire, nodeColor=<rgba 0 0.5078 1 1>}
  n34 = MaterialFunctionCall!(CustomUPosition) {materialFunction=/SparksStageShow/Systems/StageManager/FXManager/MaterialFunctions/MF_MojoStageColors}
  n35 = ScalarParameter() {parameterName=UseStageColor}
  NamedRerouteDeclaration(None=LinearInterpolate(A=VectorParameter() {parameterName=Tertiary Color, defaultValue=<rgba 1 0.0031 0.1232 1>}.RGB, B=@n34.Tertiary, Alpha=@n35)) {name=TertiaryColor, nodeColor=<rgba 0.3125 0 1 1>}
  n36 = SmoothStep(
   Min=ScalarParameter() {defaultValue=0.55, parameterName=FireFront_Gradient_Min}
   Max=ScalarParameter() {defaultValue=0.9, parameterName=FireFront_Gradient_Max}
   Value=ComponentMask(None=TextureCoordinate()) {g=true}
  ) {constMax=0.8, constMin=0.6}
  n37 = NamedRerouteUsage()
  n38 = LinearInterpolate(A=Multiply(A=ScalarParameter() {defaultValue=-1, parameterName=FireFront_Black}, B=@n37), B=@n37, Alpha=NamedRerouteUsage())
  n39 = TextureCoordinate()
  n40 = Multiply(
   A=SmoothStep(
    Min=ScalarParameter() {defaultValue=0.5, parameterName=FireFront_Glow_Min}
    Max=ScalarParameter() {defaultValue=1, parameterName=FireFront_Glow_Max}
    Value=ComponentMask(None=@n39) {g=true}
   ) {constMax=0.8, constMin=0.6}
   B=ScalarParameter() {defaultValue=1, parameterName=FireFront_Glow_Power}
  )
  n41 = Multiply(A=@n37, B=@n40) {desc=Remove for White Glow}
  n42 = Multiply(A=@n40, B=ScalarParameter() {parameterName=FireFront_White}) {constB=0.01}
  NamedRerouteDeclaration(None=Add(A=Multiply(A=@n36, B=@n38), B=Add(A=@n41, B=@n42))) {name=FG_Fire, nodeColor=<rgba 1 0.2109 0 1>}
  NamedRerouteDeclaration(None=LinearInterpolate(A=VectorParameter() {parameterName=PrimaryColor, defaultValue=<rgba 0.8125 0 0.0995 1>}.RGB, B=@n34.Primary, Alpha=@n35)) {name=PrimaryColor, nodeColor=<rgba 0 0 0 1>}
  NamedRerouteDeclaration(None=LinearInterpolate(A=VectorParameter() {parameterName=SecondaryColor, defaultValue=<rgba 1 0.1774 0 1>}.RGB, B=@n34.Secondary, Alpha=@n35)) {name=SecondaryColor, nodeColor=<rgba 0 0 0 1>}
  n43 = LinearInterpolate(A=Panner() {speedX=-0.05, speedY=-0.125}, B=Panner() {speedX=-0.1, speedY=-0.25}, Alpha=ScalarParameter() {parameterName=CastleWindows?})
  NamedRerouteDeclaration(None=LinearInterpolate(
    A=NamedRerouteUsage()
    B=NamedRerouteUsage()
    Alpha=TextureSample(UVs=@n43) {texture=/Game/Effects/Fort_Effects/Textures/Tile/T_GaussianNoise_01_M}.RGB
   )) {name=TerrainColorWash, nodeColor=<rgba 0 1 0.3906 1>}
  NamedRerouteDeclaration(None=Saturate(None=Add(A=@n14.G, B=@n14.G))) {name=CastleMaskCenter, nodeColor=<rgba 1 0.2812 0 1>}
  n44 = Constant() {desc=BeatLow, r=1.2}
  n45 = ScalarParameter() {defaultValue=1, parameterName=Scale X}
  n46 = ComponentMask(None=TextureCoordinate()) {g=true}
  n47 = ScalarParameter() {defaultValue=1, parameterName=Flame Top UV Stretch}
  n48 = Add(A=Multiply(A=ScalarParameter() {defaultValue=1, parameterName=Scale Y}, B=Multiply(A=@n46, B=@n47))) {constB=0.5}
  n49 = Panner(
   Coordinate=Multiply(A=TextureCoordinate() {uTiling=2, vTiling=2}, B=AppendVector(A=@n45, B=@n48))
   Time=Time()
   Speed=AppendVector(A=ScalarParameter() {defaultValue=-0.05, parameterName=Speed X}, B=ScalarParameter() {defaultValue=0.15, parameterName=Speed Y})
  )
  n50 = ScalarParameter() {defaultValue=1, parameterName=Fire Tile 1 U}
  n51 = ScalarParameter() {defaultValue=1, parameterName=Fire Tile 1 V}
  n52 = Panner(
   Coordinate=Multiply(A=TextureCoordinate(), B=AppendVector(A=@n50, B=@n51))
   Time=Time()
   Speed=AppendVector(A=ScalarParameter() {defaultValue=-0.05, parameterName=Speed X}, B=ScalarParameter() {defaultValue=0.15, parameterName=Speed Y})
  )
  n53 = SmoothStep(
   Min=ScalarParameter() {parameterName=SSMin}
   Max=ScalarParameter() {defaultValue=0.8, parameterName=SSMax}
   Value=LinearInterpolate(
    A=TextureSample(UVs=@n49) {texture=/Game/Effects/Fort_Effects/Textures/T_WispyFireNoise03}.R
    B=TextureSample(UVs=@n52) {texture=&T_WispyFireNoise03}.R
    Alpha=ScalarParameter() {defaultValue=0.5, parameterName=Use Dive Pan Speed}
   )
  ) {constMin=0.1}
  n54 = Constant() {desc=BeatHigh, r=0.5}
  NamedRerouteDeclaration(None=LinearInterpolate(
    A=Multiply(A=@n44, B=@n53) {constB=1.2}
    B=Multiply(A=@n53, B=@n54)
    Alpha=CollectionParameter() {collection=&MPC_Sparks_Audio_Parameters, parameterName=BeatLifetime}
   )) {name=Fire, nodeColor=<rgba 0 0.6719 1 1>}
  TextureSample() {samplerType=SAMPLERTYPE_LinearGrayscale, texture=/Game/Global/Textures/Default/DefaultUI/Checkerboard_D}
 ]
}
